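; Passive quest ability that tracks the charge of the high-tech backpack in a Battle Pass Season 20 narrative quest.
; On the server, an energy cell tag adds capped charge and restarts a looping timer that drains one bar per interval.
; At half charge or above it applies a drop-off-eligibility effect, and first reaching maximum triggers a VO effect.
; Cosmetic updates drive the backpack part modifier's charge bar, though the OnRep handler passes zero.
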(event Ability|EventActivateAbility
  ;[0 0 0] Activate passive ability - wait for the player to pick up the first energy cell before beginning the drop charge timer
  (|CallParentFunction)
  (if (Networking|IsServer)
    (Ability|Tasks|WaitGameplayTagAdd (Variables|Quest|GetCollectedEnergyCellTag)
      (:Added
        (CallFunction|CellCollected (Variables|Quest|GetChargeBarInterval))))))

(event Custom|CellCollected
  (params (ChargeBarInterval :type Float%20%28double-precision%29))
  ;[0 0 0] CellCollected - restore some charge to the pack and reset timer
  (Development|PrintString "Backpack: Interacted with Energy Cell" true true "(R=1.000000,G=0.000000,B=0.529412,A=1.000000)")
  (Variables|BackpackDisplay|SetChargeBarAmount (+ (Variables|BackpackDisplay|GetChargeBarAmount) ChargeBarInterval))
  (if (>= (Variables|BackpackDisplay|GetChargeBarAmount) (Variables|Quest|GetChargeBarMaxValue))
    (Variables|BackpackDisplay|SetChargeBarAmount (Variables|Quest|GetChargeBarMaxValue))
    (if (not (Variables|Default|GetbMaxChargeReached))
      (Variables|Default|SetbMaxChargeReached true)
      (GameplayEffects|ApplyGameplayEffectToSelf :self (Ability|GetFortAbilitySystemComponent (CallFunction|GetAvatarasFortPlayerPawn)) :GameplayEffectClass (Variables|Default|GetDropoffPackVO) :Level 0.0))
    (else
      _))
  (CallFunction|UpdatePartModifier (Variables|BackpackDisplay|GetChargeBarAmount))
  (CallFunction|ResetTimer)
  (CallFunction|HandlePackPercentageCheck))

(event Custom|ResetTimer
  ;[0 0 0] ResetTimer - Helper event to reset the timer
  (bind _returnvalue (Utilities|Time|SetTimerbyEvent (delegate Custom|DropCharge) (Variables|Quest|GetTimeUntilChargeDrop) true))
  (Variables|Quest|SetChargeDropTimerHandle _returnvalue))

(event Custom|DropCharge
  ;[0 0 0] DropCharge - When the specified amount of time has passed since last picking up a cell, drop the charge by 1 bar in the backpack
  (if (> (Variables|BackpackDisplay|GetChargeBarAmount) 0)
    (bind _output_get (Variables|BackpackDisplay|SetChargeBarAmount (- (Variables|BackpackDisplay|GetChargeBarAmount) 0.076923)))
    (Development|PrintString (Utilities|String|Append "High Tech Pack: Charge Dropped. Current charge is " (Utilities|String|ToString(Float) _output_get)) true true "(R=1.000000,G=0.000000,B=0.529412,A=1.000000)" 3.0)
    (CallFunction|UpdatePartModifier _output_get)
    (CallFunction|HandlePackPercentageCheck)
    (else
      (Development|PrintString "High Tech Pack is empty" true true "(R=1.000000,G=0.000000,B=0.529412,A=1.000000)" 3.0))))

(event Custom|HandlePackPercentageCheck
  (bind _returnvalue (CallFunction|GetAvatarasFortPlayerPawn))
  (bind _returnvalue_2 (Ability|GetFortAbilitySystemComponent _returnvalue))
  ;[0 0 0] If the pack is above a certain charge apply a tag to the player that allows them to drop it off
  (if (Networking|IsServer)
    (if (>= (Variables|BackpackDisplay|GetChargeBarAmount) 0.5)
      (bind _returnvalue_3 (GameplayEffects|ApplyGameplayEffectToSelf :self _returnvalue_2 :GameplayEffectClass "/BattlePassS20/Items/QuestAbilities/Quest_Narrative_ChargeBackpack/GE_Quest_HightechPack_CanDropOff.GE_Quest_HightechPack_CanDropOff_C" :Level 0.0))
      (Variables|Quest|SetGE_CanDropOffPack _returnvalue_3)
      (Development|PrintString "Backpack: Can be dropped off" true true "(R=1.000000,G=0.000000,B=0.529412,A=1.000000)")
      (else
        (GameplayEffects|RemoveActiveGameplayEffect _returnvalue_2 (Variables|Quest|GetGE_CanDropOffPack) -1)
        (Development|PrintString "Backpack: Less than 50%, can't drop off!" true true "(R=1.000000,G=0.000000,B=0.529412,A=1.000000)")))))

(event Custom|UpdatePartModifier
  (params (ChargeAmount :type Float%20%28double-precision%29))
  ;[0 0 0] Client only - update the material on the charge bar when the charge value is changed
  (Networking|CanExecuteCosmeticEvents
    (:True
      (bind _asb_athena_part_modifier_high_tech_backpack (Utilities|Casting|CastToB_Athena_PartModifier_HighTechBackpack (FortPawn|GetModifierforPartType (CallFunction|GetAvatarasFortPlayerPawn) "Backpack")))
      (Class|BAthenaPartModifierHighTechBackpack|UpdateChargeBarScalar _asb_athena_part_modifier_high_tech_backpack ChargeAmount))))

(event Ability|EventOnEndAbility
  (params (bWasCancelled :type Boolean))
  (bind _chargedroptimerhandle (Variables|Quest|GetChargeDropTimerHandle))
  ;[0 0 0] On End Ability - clear out and invalidate the update timer and unbind any currently bound events
  (|CallParentFunction bWasCancelled)
  (if (Utilities|Time|IsValidTimerHandle _chargedroptimerhandle)
    (Utilities|Time|ClearandInvalidateTimerbyHandle _chargedroptimerhandle)))

(event Ability|EventFailedToActivatePassiveAbility
  (params (ActorInfo :type Gameplay%20Ability%20Actor%20Info%20Structure))
  (|CallParentFunction))

(fn OnRep_ChargeBarAmount
  (Development|PrintString "Charge bar changed" true true "(R=1.000000,G=0.000000,B=0.179738,A=1.000000)" 4.0)
  ;[0 0 0] Client only - update the material on the charge bar when the value is changed
  (CallFunction|UpdatePartModifier 0.0))
